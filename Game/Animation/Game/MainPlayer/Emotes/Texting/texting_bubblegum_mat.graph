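# Simple material for the bubblegum in the main player's texting emote, going by its name and folder.
# It uses the Default Lit shading model with legacy pixel depth offset mode.
# Its only node is a vector parameter defaulting to bubblegum pink, with no link to a material output shown in the graph.
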Material /Game/Animation/Game/MainPlayer/Emotes/Texting/texting_bubblegum_mat
graph {
 properties {pixelDepthOffsetMode=PDOM_Legacy, usedShadingModels=Default Lit}
 nodes [VectorParameter() {defaultValue=<rgba 1 0.429 0.6818 0>}]
}
